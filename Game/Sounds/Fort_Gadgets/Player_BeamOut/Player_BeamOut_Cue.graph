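# Sound cue for the player beam-out gadget sound, reusing a teleporter placement wave.
# A single wave player feeds the output directly at reduced volume, with no randomization or modulation.
# It uses the gadget sound class and shared gadget-activate attenuation.
# An overridden concurrency rule prevents new instances from starting once the limit is reached.
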SoundCue /Game/Sounds/Fort_Gadgets/Player_BeamOut/Player_BeamOut_Cue
root=WavePlayer_4
nodes [{id=WavePlayer_4, properties {soundWaveAssetPtr=/Game/Sounds/Fort_Gadgets/Teleporter/teleporter_place_02}}]
properties {
 attenuationSettings=&Fort_GadgetActivate_Att
 bOverrideConcurrency=true
 maxDistance=4000
 soundClassObject=/Game/Sounds/FortSoundClassesAndModes/Fort_Gadgets_StW
 volumeMultiplier=0.6
 concurrencyOverrides {resolutionRule=PreventNew}
}
